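; Living World spawn action that applies a configurable gameplay tag to spawned Fortnite Athena vehicles.
; On all actors spawned, it casts each actor to a vehicle and adds TagToAdd as a replicated loose gameplay tag.
; If the add fails, it writes a shipping log message naming the tag and the cast vehicle.
; The source point provider is not used.
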(fn OnAllActorSpawned
  (params (SpawnedActor :type Array%20of%20Actor%20Object%20References) (SourcePointProvider :type Fort%20Athena%20Living%20World%20Point%20Provider%20Interface%20Interface%20%28by%20ref%29))
  (bind _tagtoadd (Variables|Default|GetTagToAdd))
  (for _array_element SpawnedActor
    (bind _asfort_athena_vehicle (Utilities|Casting|CastToFortAthenaVehicle _array_element))
    (bind _returnvalue (Ability|GameplayEffect|AddLooseGameplayTags _asfort_athena_vehicle (GameplayTags|MakeGameplayTagContainerfromTag _tagtoadd) true))
    (if (not _returnvalue)
      (Fortnite|FortShippingLog self (Utilities|String|ToString(Text) (Utilities|Text|FormatText "NSLOCTEXT(\"[F147BB5FFB8310DDC7C5D67DDEC2B807]\", \"BCC3216F483B947F428E1B86DF668D2A\", \"Trying to add tag: {a} to actor {b} but failed to Add Loose Gameplay Tags.\")" (GameplayTags|GetDebugStringfromGameplayTag _tagtoadd) _asfort_athena_vehicle)) true)))
  (return))
